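# Physics asset for the Junk Samurai backpack cosmetic, simulating the flag mounted on the pack.
# Capsule bodies cover the pack, a thin flag base and six flag segments, constrained in a single chain from pack to flag tip.
# Swing1 and twist limits start tight near the base and widen sharply toward the tip, while swing2 stays very narrow along the flag, so it bends mostly in one plane.
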PhysicsAsset /BRCosmetics/Accessories/FORT_Backpacks/Backpack_M_MED_Junk_Samurai/Meshes/M_MED_Junk_Samurai_Pack_Physics1
bodies {
 dyn_flagBase [{shapeType=Capsule, length=22, radius=1, rotation=<rot -7.3558 3.0475 -14.0441>}]
 dyn_flag_01 [{shapeType=Capsule, length=9, radius=4, center=<1.0838 0 -0.5668>}]
 dyn_flag_02 [{shapeType=Capsule, length=7, radius=5, center=<1.2933 0 -0.2203>, rotation=<rot -2.6535 -0.0002 0.0001>}]
 dyn_flag_03 [{shapeType=Capsule, length=7, radius=5, center=<0 0 -0.6976>}]
 dyn_flag_04 [{shapeType=Capsule, length=9, radius=3}]
 dyn_flag_05 [{shapeType=Capsule, length=9, radius=3}]
 dyn_flag_06 [{shapeType=Capsule, length=9, radius=3}]
 pack [{shapeType=Capsule, length=57.9026, radius=1.2365, center=<3.0258 -11.2572 32.2232>, rotation=<rot -8.6697 -88.3917 4.8492>}]
}
constraints [
 {bone1Name=dyn_flag_02, bone2Name=dyn_flag_01, swing1LimitDegrees=3, swing2LimitDegrees=1, twistLimitDegrees=4}
 {bone1Name=dyn_flag_03, bone2Name=dyn_flag_02, swing1LimitDegrees=3, swing2LimitDegrees=1, twistLimitDegrees=4}
 {bone1Name=dyn_flag_04, bone2Name=dyn_flag_03, swing1LimitDegrees=20, swing2LimitDegrees=1, twistLimitDegrees=10}
 {bone1Name=dyn_flag_05, bone2Name=dyn_flag_04, swing1LimitDegrees=50, swing2LimitDegrees=1, twistLimitDegrees=20}
 {bone1Name=dyn_flag_06, bone2Name=dyn_flag_05, swing1LimitDegrees=60, swing2LimitDegrees=1, twistLimitDegrees=30}
 {bone1Name=dyn_flag_01, bone2Name=dyn_flagBase, swing1LimitDegrees=3, swing2LimitDegrees=1, twistLimitDegrees=3}
 {bone1Name=dyn_flagBase, bone2Name=pack, swing1LimitDegrees=5, swing2LimitDegrees=5, twistLimitDegrees=5}
]
