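# Animation Blueprint for the Junk Samurai backpack cosmetic, posing the pack and adding secondary motion to its flag, zipper and tassels.
# Player-state booleans pick additive pack-bone offsets feeding two RigidBody branches, blended on jumping, falling or emoting.
# Position constraints with decreasing alphas restrain the simulated flag chain.
# Anim Dynamics then swings zipper and tassels before final flag constraints and a pack offset; the event graph has no logic.
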AnimBlueprint /BRCosmetics/Accessories/FORT_Backpacks/Backpack_M_MED_Junk_Samurai/Meshes/M_MED_Junk_Samurai_Pack_AnimBP
target_skeleton=&/M_MED_Junk_Samurai_Pack_Skeleton
graphs [
 AnimGraph {nodes [
   n1 = "Transform (Modify) Bone - Bone: pack"(ComponentPose=Local To Component(LocalPose=Local Space Ref Pose()), Translation="0,0,0", Rotation="0,0,10") {
    rotationMode=BMM_Additive
    scaleMode=BMM_Additive
    translationMode=BMM_Additive
    alphaBoolBlend {blendOption=Linear}
    boneToModify {boneName=pack}
    rotation {roll=10}
   }
   n2 = Blend Poses by bool(
    BlendPose_0=Component To Local(ComponentPose=@n1)
    BlendPose_1=Local Space Ref Pose()
    bActiveValue=OR Boolean!(A=Get bIsJumping(), B=Get bIsFalling(), C:bool=Get bIsSprinting()) -> (ReturnValue)
   ) {blendType=Linear}
   n3 = "Transform (Modify) Bone - Bone: pack"(ComponentPose=Local To Component(LocalPose=Local Space Ref Pose()), Translation="-5,1,0", Rotation="0,0,0") {
    rotationMode=BMM_Additive
    scaleMode=BMM_Additive
    translationMode=BMM_Additive
    alphaBoolBlend {blendOption=Linear}
    boneToModify {boneName=pack}
    translation=<-5 1>
   }
   n4 = Component To Local(ComponentPose=@n3)
   n5 = Get bIsTargeting()
   n6 = "Transform (Modify) Bone - Bone: pack"(
    ComponentPose=Local To Component(LocalPose=Blend Poses by bool(BlendPose_0=@n4, BlendPose_1=@n2, bActiveValue=@n5) {blendType=Linear})
    Translation="0,-0.3,0"
    Rotation="0,0,8"
   ) {
    rotationMode=BMM_Additive
    scaleMode=BMM_Additive
    translationMode=BMM_Additive
    alphaBoolBlend {blendOption=Linear}
    boneToModify {boneName=pack}
    rotation {roll=8}
    translation {y=-0.3}
   }
   n7 = RigidBody(ComponentPose=@n6) {
    bClampLinearTranslationLimitToRefPose=true
    bOverrideWorldGravity=true
    overridePhysicsAsset=&/M_MED_Junk_Samurai_Pack_Physics1
    alphaBoolBlend {blendOption=Linear}
    baseBoneRef {boneName=root}
    componentLinearAccScale=<0.35 0.35 0.35>
    componentLinearVelScale=<0.35 0.35 0.2>
    overrideWorldGravity {z=-980}
   }
   n8 = "Constraint - dyn_flag_01 (P)"(ComponentPose=@n7, Alpha=0.7, ConstraintWeights_0=1) {
    alphaBoolBlend {blendOption=Linear}
    boneToModify {boneName=dyn_flag_01}
    constraintSetup [{targetBone {boneName=dyn_flagBase}}]
    constraintWeights [1]
   }
   n9 = "Constraint - dyn_flag_02 (P)"(ComponentPose=@n8, Alpha=0.5, ConstraintWeights_0=1) {
    alphaBoolBlend {blendOption=Linear}
    boneToModify {boneName=dyn_flag_02}
    constraintSetup [{targetBone {boneName=dyn_flag_01}}]
    constraintWeights [1]
   }
   n10 = "Constraint - dyn_flag_03 (P)"(ComponentPose=@n9, Alpha=0.4, ConstraintWeights_0=1) {
    alphaBoolBlend {blendOption=Linear}
    boneToModify {boneName=dyn_flag_03}
    constraintSetup [{targetBone {boneName=dyn_flag_02}}]
    constraintWeights [1]
   }
   n11 = "Constraint - dyn_flag_04 (P)"(ComponentPose=@n10, Alpha=0.3, ConstraintWeights_0=1) {
    alphaBoolBlend {blendOption=Linear}
    boneToModify {boneName=dyn_flag_04}
    constraintSetup [{targetBone {boneName=dyn_flag_03}}]
    constraintWeights [1]
   }
   n12 = "Constraint - dyn_flag_05 (P)"(ComponentPose=@n11, Alpha=0.2, ConstraintWeights_0=1) {
    alphaBoolBlend {blendOption=Linear}
    boneToModify {boneName=dyn_flag_05}
    constraintSetup [{targetBone {boneName=dyn_flag_04}}]
    constraintWeights [1]
   }
   n13 = "Constraint - dyn_flag_06 (P)"(ComponentPose=@n12, Alpha=0.1, ConstraintWeights_0=1) {
    alphaBoolBlend {blendOption=Linear}
    boneToModify {boneName=dyn_flag_06}
    constraintSetup [{targetBone {boneName=dyn_flag_05}}]
    constraintWeights [1]
   }
   n14 = Component To Local(ComponentPose=@n13)
   n15 = "Transform (Modify) Bone - Bone: pack"(ComponentPose=Local To Component(LocalPose=Local Space Ref Pose()), Translation="0,0.5,0", Rotation="0,0,10") {
    rotationMode=BMM_Additive
    scaleMode=BMM_Additive
    translationMode=BMM_Additive
    alphaBoolBlend {blendOption=Linear}
    boneToModify {boneName=pack}
    rotation {roll=10}
    translation {y=0.5}
   }
   n16 = Component To Local(ComponentPose=@n15)
   n17 = "Transform (Modify) Bone - Bone: pack"(ComponentPose=Local To Component(LocalPose=Local Space Ref Pose()), Translation="-0.5,1,0", Rotation="0,-2,10") {
    rotationMode=BMM_Additive
    scaleMode=BMM_Additive
    translationMode=BMM_Additive
    alphaBoolBlend {blendOption=Linear}
    boneToModify {boneName=pack}
    rotation {roll=10, yaw=-2}
    translation=<-0.5 1>
   }
   n18 = Component To Local(ComponentPose=@n17)
   n19 = Get bIsTargeting()
   n20 = SelectFloat(A=0.5, B=0.5, bPickA=Get bIsSprinting())
   n21 = OR Boolean(A=Get bIsJumping(), B=Get bIsFalling(), C=Get bIsPlayingEmote())
   n22 = Get bIsRelaxedLevel1() -> (bIsRelaxedLevel1:bool)
   n23 = Get bIsRelaxedLevel2() -> (bIsRelaxedLevel2:bool)
   n24 = RigidBody(
    ComponentPose=Local To Component(LocalPose=Blend Poses by bool(BlendPose_0=@n18, BlendPose_1=@n16, bActiveValue=@n19) {blendType=Linear})
    Alpha=SelectFloat(A=0.5, B=SelectFloat(A=0.2, B=@n20, bPickA=@n21), bPickA=OR Boolean(A=@n22, B=@n23))
   ) {
    bClampLinearTranslationLimitToRefPose=true
    bOverrideWorldGravity=true
    overridePhysicsAsset=&M_MED_Junk_Samurai_Pack_Physics1
    alphaBoolBlend {blendOption=Linear}
    componentLinearAccScale=<1.5 1.5 1.5>
    componentLinearVelScale=<1.5 1.5 1.5>
    overrideWorldGravity {z=-980}
   }
   n25 = "Constraint - dyn_flag_01 (P)"(ComponentPose=@n24, Alpha=0.3, ConstraintWeights_0=1) {
    alphaBoolBlend {blendOption=Linear}
    boneToModify {boneName=dyn_flag_01}
    constraintSetup [{targetBone {boneName=dyn_flagBase}}]
    constraintWeights [1]
   }
   n26 = "Constraint - dyn_flag_02 (P)"(ComponentPose=@n25, Alpha=0.2, ConstraintWeights_0=1) {
    alphaBoolBlend {blendOption=Linear}
    boneToModify {boneName=dyn_flag_02}
    constraintSetup [{targetBone {boneName=dyn_flag_01}}]
    constraintWeights [1]
   }
   n27 = Component To Local(ComponentPose=@n26)
   n28 = OR Boolean(A=Get bIsJumping(), B=Get bIsFalling(), C=Get bIsPlayingEmote())
   n29 = Anim Dynamics - Bone: dyn_zipper(
    ComponentPose=Local To Component(LocalPose=Blend Poses by bool(BlendPose_0=@n14, BlendPose_1=@n27, BlendTime_0=0.3, BlendTime_1=0.2, bActiveValue=@n28) {blendType=Linear})
    Alpha=1
   ) {
    angularDampingOverride=0.95
    bGravityOverrideInSimSpace=true
    bOverrideAngularDamping=true
    simulationSpace=RootRelative
    alphaBoolBlend {blendOption=Linear}
    boundBone {boneName=dyn_zipper}
    componentLinearAccScale {y=0.6}
    physicsBodyDefinitions [{
      sphereCollisionRadius=0
      boundBone {boneName=dyn_zipper}
      boxExtents=<3 10 10>
      constraintSetup {angularLimitsMax=<2 37.1656 45>, angularLimitsMin=<-2 -23.758 -45>, angularTarget=<0 0 0>}
      localJointOffset=<1.5 -0.05 0.1>
     }]
   }
   n30 = Anim Dynamics - Bone: dyn_tassleBase(ComponentPose=@n29, Alpha=1, ExternalForce="0,0,0") {
    angularDampingOverride=0.99
    bGravityOverrideInSimSpace=true
    bOverrideAngularDamping=true
    bUseGravityOverride=true
    alphaBoolBlend {blendOption=Linear}
    boundBone {boneName=dyn_tassleBase}
    componentAppliedLinearAccClamp=<3000 3000 3000>
    componentLinearAccScale=<0.1 0.2 0.4>
    gravityOverride {z=-980}
    physicsBodyDefinitions [{
      sphereCollisionRadius=0
      boundBone {boneName=dyn_tassleBase}
      boxExtents=<4 10 10>
      constraintSetup {angularLimitsMax=<10 18.344 56.0319>, angularLimitsMin=<-10 -81.4014 -69.7899>, angularTarget=<0 0 0>}
      localJointOffset=<0.9 0 0>
     }]
   }
   n31 = Anim Dynamics - Bone: dyn_tassle_01(ComponentPose=@n30, Alpha=1, ExternalForce="0,0,0") {
    angularDampingOverride=0.99
    bGravityOverrideInSimSpace=true
    bOverrideAngularDamping=true
    bUseGravityOverride=true
    alphaBoolBlend {blendOption=Linear}
    boundBone {boneName=dyn_tassle_01}
    componentAppliedLinearAccClamp=<3000 3000 3000>
    componentLinearAccScale=<0.1 0.2 0.1>
    gravityOverride {z=-980}
    physicsBodyDefinitions [{
      sphereCollisionRadius=0
      boundBone {boneName=dyn_tassle_01}
      boxExtents=<4 10 10>
      constraintSetup {angularLimitsMax=<10 -9.172 56.0319>, angularLimitsMin=<-10 -81 -69.7899>, angularTarget=<0 0 0>}
      localJointOffset=<0.9 0 0>
     }]
   }
   n32 = Anim Dynamics - Bone: dyn_tassle_l(ComponentPose=@n31, Alpha=0.5, ExternalForce="0,0,0") {
    angularDampingOverride=0.85
    bGravityOverrideInSimSpace=true
    bOverrideAngularDamping=true
    bUseGravityOverride=true
    alphaBoolBlend {blendOption=Linear}
    boundBone {boneName=dyn_tassle_l}
    componentAppliedLinearAccClamp=<3000 3000 3000>
    componentLinearAccScale=<0.1 0.2 0.1>
    gravityOverride {z=-980}
    physicsBodyDefinitions [{
      sphereCollisionRadius=0
      boundBone {boneName=dyn_tassle_l}
      boxExtents=<4 10 10>
      constraintSetup {angularLimitsMax=<1 5 56.0319>, angularLimitsMin=<-1 -81.4014 -69.7899>, angularTarget=<0 0 0>}
      localJointOffset=<0.9 0 0>
     }]
   }
   n33 = Anim Dynamics - Bone: dyn_tassle_r(ComponentPose=@n32, Alpha=0.5, ExternalForce="0,0,0") {
    angularDampingOverride=0.85
    bGravityOverrideInSimSpace=true
    bOverrideAngularDamping=true
    bUseGravityOverride=true
    alphaBoolBlend {blendOption=Linear}
    boundBone {boneName=dyn_tassle_r}
    componentAppliedLinearAccClamp=<3000 3000 3000>
    componentLinearAccScale=<0.1 0.2 0.1>
    gravityOverride {z=-980}
    physicsBodyDefinitions [{
      sphereCollisionRadius=0
      boundBone {boneName=dyn_tassle_r}
      boxExtents=<4 10 10>
      constraintSetup {angularLimitsMax=<1 5 56.0319>, angularLimitsMin=<-1 -81.4014 -69.7899>, angularTarget=<0 0 0>}
      localJointOffset=<0.9 0 0>
     }]
   }
   n34 = "Constraint - dyn_flag_02 (P)"(ComponentPose=@n33, Alpha=0.2, ConstraintWeights_0=1) {
    alphaBoolBlend {blendOption=Linear}
    boneToModify {boneName=dyn_flag_02}
    constraintSetup [{targetBone {boneName=dyn_flag_01}}]
    constraintWeights [1]
   }
   n35 = "Constraint - dyn_flag_01 (P)"(ComponentPose=@n34, Alpha=0.2, ConstraintWeights_0=1) {
    alphaBoolBlend {blendOption=Linear}
    boneToModify {boneName=dyn_flag_01}
    constraintSetup [{targetBone {boneName=dyn_flagBase}}]
    constraintWeights [1]
   }
   n36 = "Transform (Modify) Bone - Bone: pack"(ComponentPose=@n35, Translation="0,1,0", Rotation="0,0,-10") {
    rotationMode=BMM_Additive
    scaleMode=BMM_Additive
    translationMode=BMM_Additive
    alphaBoolBlend {blendOption=Linear}
    boneToModify {boneName=pack}
    rotation {roll=-10}
    translation {y=1}
   }
   Output Pose(Result=Component To Local(ComponentPose=@n36)) {layerGroup=None}
  ]}
 EventGraph {nodes [
   TryGetPawnOwner()
   Event BlueprintUpdateAnimation()
  ]}
]
variables []
